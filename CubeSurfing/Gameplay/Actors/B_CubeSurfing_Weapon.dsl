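; Cube Surfing weapon Blueprint, a held magic cube that manages its HUD, effects, audio and ammo while the player surfs.
; On equip, async tasks watch the surf tag and boost event to switch surf state and apply boost cooldown.
; A timer drains ammo at boost-dependent rates, stopping surfing on an empty magazine and removing the item when ammo runs out.
; Timelines drive cube material transitions, with the deploy intro replicated via OnRep.
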(event Custom|ShowBoostCooldown
  ;[0.01376 0.2962 1] Set up asyncs to monitor weapon state
  (GameplayEffects|ApplyGameplayEffectToSelf :self (Ability|GetFortAbilitySystemComponent (Variables|Default|GetPlayerPawn)) :GameplayEffectClass "/CubeSurfing/Gameplay/Effects/GE_CubeSurfing_BoostCooldown.GE_CubeSurfing_BoostCooldown_C" :Level 0.0))

(event Custom|CubeSurfActivate
  ;[1 0.4563 0.04423] Activate / Deactivate Cube Surfing
  (CallFunction|ToggleWeaponHUD true)
  (CallFunction|ToggleWeaponFX true)
  (Components|Activation|Deactivate (Variables|Default|GetIdle))
  (Class|WeaponEquipSoundComponent|StopAudio (Variables|Default|GetWeaponEquipSoundComponent))
  (SoundIndicator|StartTargetedSoundIndicatorLooping (Variables|SoundIndicator|GetSoundIndicatorComponent) self "/CubeSurfing/Assets/Audio/MSS_CubeSurfing_Surf_Idle_LP_3P.MSS_CubeSurfing_Surf_Idle_LP_3P" 1.0 (Variables|/Script/Engine.Pawn|GetPlayerState (Variables|Default|GetPlayerPawn))))

(event FortWeapon|EventOnUnEquippedAndDetached
  (bind _magiccube (Variables|Default|GetMagicCube))
  (bind _magiccube_2 (Variables|Default|GetMagicCube))
  ;[0.01376 0.2962 1] Set up asyncs to monitor weapon state
  (Ability|Async|EndAction (Variables|Default|GetWaitTagAddAsync))
  (Ability|SendGameplayEventtoActor (Variables|Actor|GetInstigator) "(TagName=\"Event.CubeSurfing.Stop\")") ; If the weapon is unequipped or destroyed we always broadcast the stop surfing event
  (Rendering|Material|SetScalarParameterValueonMaterials _magiccube_2 "VtxHide" 0.0)
  (Rendering|Material|SetScalarParameterValueonMaterials _magiccube_2 "IsDeploying" 0.0)
  (Rendering|Material|SetScalarParameterValueonMaterials _magiccube "VtxHide_Move" 0.0)
  (Rendering|Material|SetScalarParameterValueonMaterials _magiccube "IsForming" 0.0)
  (Components|Activation|Deactivate (Variables|Default|GetIdle)))

(event FortWeapon|EventOnEquippedAndAttached
  (bind _cubesurftag (Variables|Default|GetCubeSurfTag))
  (bind _magiccube (Variables|Default|GetMagicCube))
  (bind _instigator (Variables|Actor|GetInstigator))
  ;[0.01376 0.2962 1] Set up asyncs to monitor weapon state
  (Utilities|FlowControl|Sequence
    (:then_0
      (bind _asynctaskproxy (Ability|Tasks|WaitGameplayTagAddToActor _instigator _cubesurftag)
        (:then
          (Variables|Default|SetWaitTagAddAsync _asynctaskproxy)
          (CallFunction|ToggleWeaponHUD false)
          (bind _asynctaskproxy_2 (Ability|Async|WaitGameplayEventToActor (Variables|Default|GetPlayerPawn) "(TagName=\"Event.CubeSurfing.Boost\")")
            (:then
              (Variables|Default|SetWaitBoostEventAsync _asynctaskproxy_2))
            (:EventReceived
              (CallFunction|ShowBoostCooldown))))
        (:Added
          (CallFunction|CubeSurfActivate)
          (bind _asynctaskproxy_3 (Ability|Tasks|WaitGameplayTagRemoveFromActor _instigator _cubesurftag)
            (:then
              (Variables|Default|SetWaitTagRemoveAsync _asynctaskproxy_3))
            (:Removed
              (CallFunction|CubeSurfDeactivate))))))
    (:then_1
      ;;[0.01376 0.2962 1] Begin Play
      (Rendering|Material|SetScalarParameterValueonMaterials _magiccube "VtxHide" 0.0)
      (Rendering|Material|SetScalarParameterValueonMaterials _magiccube "IsDeploying" 0.0)
      (Rendering|Material|SetScalarParameterValueonMaterials (Variables|Default|GetMagicCube) "VtxHide_Move" 0.0)
      (bind _asfort_player_pawn (Utilities|Casting|CastToFortPlayerPawn (Game|GetInstigator)))
      (Variables|Default|SetPlayerPawn _asfort_player_pawn)
      (EventDispatchers|BindEventtoOnConsumeAmmo (EventDispatchers|CreateEvent :!event DestroyWeaponOnNoAmmo)))
    (:then_2
      (Rendering|Material|SetScalarParameterValueonMaterials (Variables|Default|GetMagicCube) "IsForming" 1.0)
      (bind (_direction _vtxhide _vtxmove) (Timeline|VtxForming PlayFromStart)
        (:Update
          (Rendering|Material|SetScalarParameterValueonMaterials (Variables|Default|GetMagicCube) "VtxHide_Move_Forming" _vtxmove))
        (:Finished
          (Rendering|Material|SetScalarParameterValueonMaterials (Variables|Default|GetMagicCube) "IsForming" 0.0))))
    (:then_3
      ;;[0 0.753 1 0.5] FX INTRO
      (Components|Activation|Activate (Variables|Default|GetNS_CubeSurfing_Summon) false)
      (bind (_direction_2 _intro) (Timeline|VFX_Intro PlayFromStart)
        (:Update
          (Niagara|SetNiagaraVariable(Float) (Variables|Default|GetIdle) "AlphaIntro" _intro))))))

(event Custom|CubeSurfDeactivate
  ;[1 0.4563 0.04423] Activate / Deactivate Cube Surfing
  (CallFunction|ToggleWeaponHUD false)
  (CallFunction|ToggleWeaponFX false)
  (Components|Activation|Activate (Variables|Default|GetIdle) false)
  (Class|WeaponEquipSoundComponent|PlayAudio (Variables|Default|GetWeaponEquipSoundComponent))
  (Variables|Default|SetIsDeploying? false)
  (SoundIndicator|StopSoundIndicatorLooping (Variables|SoundIndicator|GetSoundIndicatorComponent))
  (Utilities|Time|ClearandInvalidateTimerbyHandle (Variables|Default|GetAbilityCostTimer)))

(event Custom|PlayWeaponIntro
  (bind _playerpawn (Variables|Default|GetPlayerPawn))
  (bind _mesh (Variables|/Script/Engine.Character|GetMesh _playerpawn))
  (bind _returnvalue (Components|SkeletalMesh|GetAnimInstance _mesh))
  (bind _magiccube (Variables|Default|GetMagicCube))
  (bind _magiccube_2 (Variables|Default|GetMagicCube))
  ;[1 0.4563 0.04423] Weapon Intro on Start Surfing
  (Rendering|Material|SetScalarParameterValueonMaterials _magiccube "IsDeploying" 1.0)
  (Rendering|Material|SetScalarParameterValueonMaterials _magiccube "IsForming" 0.0)
  (Rendering|Material|SetScalarParameterValueonMaterials _magiccube "DeployTime" (Utilities|Time|GetGameTimeinSeconds))
  (Timeline|TL_CubeIntro PlayFromStart
    (:Update
      (Rendering|Material|SetScalarParameterValueonMaterials _magiccube_2 "VtxHide_Deploying" (Animation|Curves|GetCurveValue _returnvalue "CubeDeploy"))
      (Rendering|Material|SetScalarParameterValueonMaterials _magiccube_2 "VtxHide_Move_Deploying" (Animation|Curves|GetCurveValue _returnvalue "CubeMove")))
    (:Finished
      (Rendering|Material|SetScalarParameterValueonMaterials _magiccube_2 "VtxHide_Deploying" 1.0))) ; This timeline should be shorter than animation montage to make sure the value is one before animation blending
)

(event EventBeginPlay
  (bind _magiccube (Variables|Default|GetMagicCube))
  ;[0.01376 0.2962 1] Set up asyncs to monitor weapon state
  ;;[0.01376 0.2962 1] Begin Play
  (Rendering|Material|SetScalarParameterValueonMaterials _magiccube "IsForming" 1.0)
  (Rendering|Material|SetScalarParameterValueonMaterials _magiccube "VtxHide" 0.0)
  (Rendering|Material|SetScalarParameterValueonMaterials _magiccube "IsDeploying" 0.0)
  (Rendering|Material|SetScalarParameterValueonMaterials (Variables|Default|GetMagicCube) "VtxHide_Move" 0.0)
  (bind _asfort_player_pawn (Utilities|Casting|CastToFortPlayerPawn (Game|GetInstigator)))
  (Variables|Default|SetPlayerPawn _asfort_player_pawn)
  (EventDispatchers|BindEventtoOnConsumeAmmo (EventDispatchers|CreateEvent :!event DestroyWeaponOnNoAmmo)))

(event EventEndPlay
  (params (EndPlayReason :type EEndPlayReason%20Enum :default "Destroyed"))
  ;[0.01376 0.2962 1] Set up asyncs to monitor weapon state
  (EventDispatchers|UnbindEventfromOnConsumeAmmo (EventDispatchers|CreateEvent :!event DestroyWeaponOnNoAmmo)))

(event Custom|StartSpendingAmmo
  (params (Rate :type Float%20%28double-precision%29) (ConsumeInstantly? :type Boolean))
  ;[0.01376 0.2962 1] Ammo consumption
  (Utilities|Time|ClearandInvalidateTimerbyHandle (Variables|Default|GetAbilityCostTimer))
  (bind _returnvalue (Utilities|Time|SetTimerbyEvent (EventDispatchers|CreateEvent :!event ConsumeCubeAmmo) Rate true true))
  (Variables|Default|SetAbilityCostTimer _returnvalue)
  (if ConsumeInstantly?
    (CallFunction|ConsumeCubeAmmo)))

(event Custom|ToggleWeaponFX
  (params (Enable :type Boolean))
  (bind _magiccube (Variables|Default|GetMagicCube))
  (bind _playerpawn (Variables|Default|GetPlayerPawn))
  (bind _mesh (Variables|/Script/Engine.Character|GetMesh _playerpawn))
  (bind _magiccube_2 (Variables|Default|GetMagicCube))
  ;[0.01376 0.2962 1] Cube Surf Visuals
  (if (not Enable)
    (Rendering|Material|SetScalarParameterValueonMaterials _magiccube_2 "IsDeploying" 0.0)
    (Rendering|Material|SetScalarParameterValueonMaterials _magiccube_2 "IsForming" 0.0)
    (bind (_direction _vtxhide _vtxmove) (Timeline|VtxAppear PlayFromStart)
      (:Update
        (Rendering|Material|SetScalarParameterValueonMaterials (Variables|Default|GetMagicCube) "VtxHide_Move" _vtxmove)
        (Rendering|Material|SetVectorParameterValueonMaterials _magiccube "FowardVector" (Math|Vector|GetRightVector (Transformation|GetSocketRotation _mesh "FX_Feet")))
        (Rendering|Material|SetVectorParameterValueonMaterials _magiccube "PawnWorldPOS" (- (Transformation|GetSocketLocation _mesh "FX_Feet") (Transformation|GetActorLocation)))
        (Rendering|Material|SetScalarParameterValueonMaterials _magiccube "VtxHide" _vtxhide)))))

(event Custom|ToggleIsBoosting
  (params (Boosting :type Boolean))
  (bind _returnvalue (Ability|GameplayCue|MakeGameplayCueParameters 0.0))
  (bind _playerpawn (Variables|Default|GetPlayerPawn))
  (bind _returnvalue_2 (GameplayTags|HasMatchingGameplayTag _playerpawn "(TagName=\"Granted.CubeSurfing.Active\")"))
  (bind _gc_boosting (Variables|Default|GetGC_Boosting))
  (bind _playerpawn_2 (Variables|Default|GetPlayerPawn))
  ;[0.01376 0.2962 1] Handle boost state
  (if Boosting
    (GameplayCue|AddGameplayCueOnActor(Looping) _playerpawn_2 _gc_boosting _returnvalue)
    (Default|CallBoosted)
    (if _returnvalue_2
      (CallFunction|StartSpendingAmmo (ScalableFloat|GetValueatLevel (Variables|Default|GetHF_AmmoRateBoosting))))
    (else
      (GameplayCue|RemoveGameplayCueOnActor(Looping) _playerpawn_2 _gc_boosting _returnvalue)
      (Default|CallNotBoosted)
      (if _returnvalue_2
        (CallFunction|StartSpendingAmmo (ScalableFloat|GetValueatLevel (Variables|Default|GetHF_AmmoRateNormal)))))))

(event Custom|ToggleIsOnWater
  (params (bIsOnWater :type Boolean))
  ;[0 0.824 1 0.5] WATER FX
  (if bIsOnWater
    (Default|CallOnWater)
    (else
      (Default|CallNotonWater))))

(event Custom|ToggleIsOnLand
  (params (bIsOnLand :type Boolean))
  ; CHECK ONLAND FX
  (if bIsOnLand
    (Default|CallOnLand)
    (else
      (Default|CallOnAir))))

(fn ConsumeCubeAmmo
  (FortWeapon|Ammo|ConsumeAmmo 1 true)
  (if (<= (FortWeapon|GetMagazineAmmoCount) 0)
    (Ability|SendGameplayEventtoActor (Variables|Default|GetPlayerPawn) "(TagName=\"Event.CubeSurfing.Stop\")")))

(fn DestroyWeaponOnNoAmmo
  (params (Quantity :type Integer))
  (if (<= (FortWeapon|GetRemainingAmmo) 0)
    (Inventory|RemoveFortItemFromPlayer (Fortnite|GetFortPlayerControllerfromActor (Variables|Default|GetPlayerPawn)) (FortWeapon|GetInventoryItem) -1 true)))

(fn OnRep_IsDeploying?
  (if (Variables|Default|GetIsDeploying?)
    (if (not (Pawn|IsLocallyControlled (Variables|Default|GetPlayerPawn)))
      (CallFunction|PlayWeaponIntro))))

(fn ToggleWeaponHUD
  (params (Surfing? :type Boolean))
  (WeaponHUDData|SetHudKeyActionVisibility 0 (not Surfing?))
  (WeaponHUDData|SetHudKeyActionVisibility 1 Surfing?)
  (WeaponHUDData|SetHudKeyActionVisibility 2 Surfing?)
  (WeaponHUDData|SetHudKeyActionVisibility 3 Surfing?)
  (WeaponHUDData|SetHudKeyActionVisibility 4 Surfing?))

(fn ConstructionScript)
